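# Physics asset for the Stink Fish gadget's skeletal mesh, giving its dynamic spine and fin bones capsule bodies for simulation.
# A small Anim body is the hub, linked to a two-segment front spine chain, a two-segment back spine chain and both side fins.
# Spine joints allow narrow swing and almost no twist, while the fins get a much wider second swing limit so they can bend further.
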PhysicsAsset /Game/Gadgets/Assets/Fish/Stink_Fish/Meshes/Stink_Fish_PhysicsAsset
bodies {
 Anim [{shapeType=Capsule, length=0.505, radius=0.505, rotation=<rot 0 -90 0>}]
 dyn_fin1_l [{shapeType=Capsule, length=4.2276, radius=1.4946, center=<6.9998 0 0.0001>, rotation=<rot 0 90 90>}]
 dyn_fin1_r [{shapeType=Capsule, length=4.2276, radius=1.4848, center=<-6.0471 -0.0001 0.0001>, rotation=<rot 0 90 90>}]
 dyn_spineBack_01 [{shapeType=Capsule, length=3.5264, radius=2.1448, center=<3.6646 -0.0017 -0.2378>, rotation=<rot 0 89.9595 -89.8215>}]
 dyn_spineBack_02 [{shapeType=Capsule, length=7.8859, radius=2.1719, center=<6.3479 -0.0039 -0.7624>, rotation=<rot 0 89.9408 -95.0112>}]
 dyn_spineFront_01 [{shapeType=Capsule, length=12.5488, radius=2.4196, center=<1.4023 0 -0.877>, rotation=<rot 0 180 -89.9995>}]
 dyn_spineFront_02 [{shapeType=Capsule, length=13.757, radius=4.2364, center=<7.7654 -0.0044 -2.5325>, rotation=<rot 0 179.9952 -89.9551>}]
}
constraints [
 {bone1Name=dyn_spineFront_01, bone2Name=Anim, swing1LimitDegrees=15, swing2LimitDegrees=12, twistLimitDegrees=2}
 {bone1Name=dyn_spineFront_02, bone2Name=dyn_spineFront_01, swing1LimitDegrees=15, swing2LimitDegrees=12, twistLimitDegrees=2}
 {bone1Name=dyn_spineBack_01, bone2Name=Anim, swing1LimitDegrees=15, swing2LimitDegrees=15, twistLimitDegrees=2}
 {bone1Name=dyn_spineBack_02, bone2Name=dyn_spineBack_01, swing1LimitDegrees=15, swing2LimitDegrees=15, twistLimitDegrees=2}
 {bone1Name=dyn_fin1_l, bone2Name=Anim, swing1LimitDegrees=15, swing2LimitDegrees=50, twistLimitDegrees=5}
 {bone1Name=dyn_fin1_r, bone2Name=Anim, swing1LimitDegrees=15, swing2LimitDegrees=50, twistLimitDegrees=5}
]
